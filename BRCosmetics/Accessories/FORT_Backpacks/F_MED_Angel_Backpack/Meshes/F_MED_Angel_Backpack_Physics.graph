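# Physics asset for the angel backpack cosmetic, defining simulation bodies and joints for its mirrored left and right wings and feathers.
# Each bone gets one box body, with wing_03 jointed to backpack_root and wing_06 to wing_03.
# Feathers A through F fan out from wing_06, and only feather A continues into a second segment.
# Narrow swing and twist limits on every joint restrict how far the wings and feathers bend.
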PhysicsAsset /BRCosmetics/Accessories/FORT_Backpacks/F_MED_Angel_Backpack/Meshes/F_MED_Angel_Backpack_Physics
bodies {
 backpack_root [{shapeType=Box, extentX=3.693, extentY=8.7117, extentZ=12.3554, center=<-1.2095 -4.4075 0>, rotation=<rot 0 -90 0>}]
 featherA_01_l [{shapeType=Box, extentX=6.2511, extentY=6.6079, extentZ=12.9932, center=<6.5601 0.6754 1.2193>, rotation=<rot 0 91.6852 -93.6835>}]
 featherA_01_r [{shapeType=Box, extentX=6.2511, extentY=6.6079, extentZ=12.9932, center=<-6.5602 -0.6754 -1.2193>, rotation=<rot 0 91.6852 -93.6835>}]
 featherA_02_l [{shapeType=Box, extentX=3.0693, extentY=24.5899, extentZ=5.7168, center=<8.8754 -0.0076 0.3405>, rotation=<rot -1.1798 -89.9998 2.5>}]
 featherA_02_r [{shapeType=Box, extentX=3.0693, extentY=24.5899, extentZ=5.7168, center=<-6.556 0.0306 -1.4605>, rotation=<rot -1.1798 -89.9998 -2.5002>}]
 featherB_01_l [{shapeType=Box, extentX=5.2102, extentY=6.4336, extentZ=16.3841, center=<7.3284 0.0005 0>, rotation=<rot 90 0 0>}]
 featherB_01_r [{shapeType=Box, extentX=2.7102, extentY=6.4336, extentZ=23.3841, center=<-9.5374 -0.0003 0>, rotation=<rot 0 90 84.9996>}]
 featherC_01_l [{shapeType=Box, extentX=5.8354, extentY=6.0193, extentZ=16.3431, center=<6.2568 0.9205 -0.245>, rotation=<rot 0 -84.366 -80.0053>}]
 featherC_01_r [{shapeType=Box, extentX=5.8354, extentY=6.0194, extentZ=16.3431, center=<-7.0089 -0.9946 0.3782>, rotation=<rot 0 -84.366 -80.0056>}]
 featherD_01_l [{shapeType=Box, extentX=3.6704, extentY=5.7968, extentZ=12.4043, center=<6.2253 0.8979 -0.4301>, rotation=<rot 0 -97.5562 -87.8146>}]
 featherD_01_r [{shapeType=Box, extentX=3.6704, extentY=5.7968, extentZ=12.4043, center=<-6.2253 -0.8979 0.4301>, rotation=<rot 0 -97.5562 -82.814>}]
 featherE_01_l [{shapeType=Box, extentX=3.441, extentY=5.4615, extentZ=13.9018, center=<5.824 0.6597 -0.8656>, rotation=<rot 0 -97.418 -88.8039>}]
 featherE_01_r [{shapeType=Box, extentX=3.441, extentY=5.4615, extentZ=13.9018, center=<-5.8246 -0.6597 0.8656>, rotation=<rot 0 -97.418 -88.804>}]
 featherF_01_l [{shapeType=Box, extentX=6.3319, extentY=6.9307, extentZ=14.1281, center=<5.2099 0.1368 0.4646>, rotation=<rot 0.0002 -89.0019 -86.9616>}]
 featherF_01_r [{shapeType=Box, extentX=6.3319, extentY=6.9307, extentZ=14.1281, center=<-5.1304 0.0047 0.4377>, rotation=<rot 0 -89.0017 -85.6523>}]
 wing_03_l [{shapeType=Box, extentX=10.108, extentY=12.2028, extentZ=29.9605, center=<9.1314 -0.7425 6.1218>, rotation=<rot 0 82.6749 -81.8478>}]
 wing_03_r [{shapeType=Box, extentX=10.108, extentY=11.7028, extentZ=29.9605, center=<-9.1314 0.7425 -6.1218>, rotation=<rot 0 82.6749 -81.8478>}]
 wing_06_l [{shapeType=Box, extentX=19.8862, extentY=10.6002, extentZ=10.7609}]
 wing_06_r [{shapeType=Box, extentX=20, extentY=9.6002, extentZ=10.7609}]
}
constraints [
 {bone1Name=wing_03_l, bone2Name=backpack_root, swing1LimitDegrees=15, swing2LimitDegrees=25, twistLimitDegrees=5}
 {bone1Name=wing_06_l, bone2Name=wing_03_l, swing1LimitDegrees=12, swing2LimitDegrees=17, twistLimitDegrees=5}
 {bone1Name=wing_03_r, bone2Name=backpack_root, swing1LimitDegrees=15, swing2LimitDegrees=25, twistLimitDegrees=5}
 {bone1Name=wing_06_r, bone2Name=wing_03_r, swing1LimitDegrees=12, swing2LimitDegrees=17, twistLimitDegrees=5}
 {bone1Name=featherA_01_l, bone2Name=wing_06_l, swing1LimitDegrees=5, swing2LimitDegrees=20, twistLimitDegrees=5}
 {bone1Name=featherA_01_r, bone2Name=wing_06_r, swing1LimitDegrees=5, swing2LimitDegrees=20, twistLimitDegrees=5}
 {bone1Name=featherA_02_l, bone2Name=featherA_01_l, swing1LimitDegrees=2, swing2LimitDegrees=18, twistLimitDegrees=5}
 {bone1Name=featherA_02_r, bone2Name=featherA_01_r, swing1LimitDegrees=2, swing2LimitDegrees=18, twistLimitDegrees=5}
 {bone1Name=featherB_01_l, bone2Name=wing_06_l, swing1LimitDegrees=10, swing2LimitDegrees=18, twistLimitDegrees=5}
 {bone1Name=featherB_01_r, bone2Name=wing_06_r, swing1LimitDegrees=10, swing2LimitDegrees=18, twistLimitDegrees=5}
 {bone1Name=featherC_01_r, bone2Name=wing_06_r, swing1LimitDegrees=10, swing2LimitDegrees=18, twistLimitDegrees=5}
 {bone1Name=featherD_01_r, bone2Name=wing_06_r, swing1LimitDegrees=10, swing2LimitDegrees=18, twistLimitDegrees=5}
 {bone1Name=featherC_01_l, bone2Name=wing_06_l, swing1LimitDegrees=10, swing2LimitDegrees=18, twistLimitDegrees=5}
 {bone1Name=featherD_01_l, bone2Name=wing_06_l, swing1LimitDegrees=10, swing2LimitDegrees=18, twistLimitDegrees=5}
 {bone1Name=featherE_01_l, bone2Name=wing_06_l, swing1LimitDegrees=10, swing2LimitDegrees=18, twistLimitDegrees=5}
 {bone1Name=featherE_01_r, bone2Name=wing_06_r, swing1LimitDegrees=10, swing2LimitDegrees=18, twistLimitDegrees=5}
 {bone1Name=featherF_01_l, bone2Name=wing_06_l, swing1LimitDegrees=8, swing2LimitDegrees=16, twistLimitDegrees=5}
 {bone1Name=featherF_01_r, bone2Name=wing_06_r, swing1LimitDegrees=8, swing2LimitDegrees=16, twistLimitDegrees=5}
]
