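# Animation Blueprint for the Angel backpack cosmetic, posing its wings by player movement state.
# A wingPosition state machine picks base, dive, fly or back poses from skydive, parachute and targeting flags, then feeds RigidBody physics with variable alpha and external force, cached as preSprint.
# Outside skydiving, additive back and base poses layer on top, weighted by SpeedAlpha or fixed weights according to crouching and targeting.
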AnimBlueprint /BRCosmetics/Accessories/FORT_Backpacks/F_MED_Angel_Backpack/Meshes/F_MED_Angel_Backpack_AnimBP
target_skeleton=/Game/Accessories/FORT_Backpacks/F_MED_Angel_Backpack/Meshes/F_MED_Angel_Backpack_Skeleton
graphs [
 AnimGraph {nodes [
   n1 = Get SpeedAlpha() -> (SpeedAlpha:float)
   n2 = F_MED_Angel_Backpack_Wings_AddPoseAsset : back!(PoseWeight:float=@n1) -> (Pose:Pose Link Structure) {
    is=title AnimGraphNode_PoseByName
    poseAsset=&/F_MED_Angel_Backpack_Wings_AddPoseAsset
    poseName=back
   }
   n3 = F_MED_Angel_Backpack_Wings_AddPoseAsset : base!(PoseWeight:float=@n1) -> (Pose:Pose Link Structure) {
    is=title AnimGraphNode_PoseByName
    poseAsset=&F_MED_Angel_Backpack_Wings_AddPoseAsset
    poseName=base
   }
   n4 = Apply Additive(Base=Apply Additive(Base=Use cached pose 'preSprint'(), Additive=@n2, Alpha=0.75) {alphaBoolBlend {blendOption=Linear}}, Additive=@n3) {alphaBoolBlend {blendOption=Linear}}
   n5 = F_MED_Angel_Backpack_Wings_AddPoseAsset : back(PoseWeight=0.05) {poseAsset=&F_MED_Angel_Backpack_Wings_AddPoseAsset, poseName=back}
   n6 = Use cached pose 'preSprint'()
   n7 = F_MED_Angel_Backpack_Wings_AddPoseAsset : base(PoseWeight=1) {poseAsset=&F_MED_Angel_Backpack_Wings_AddPoseAsset, poseName=base}
   n8 = F_MED_Angel_Backpack_Wings_AddPoseAsset : back(PoseWeight=1) {poseAsset=&F_MED_Angel_Backpack_Wings_AddPoseAsset, poseName=back}
   n9 = Use cached pose 'preSprint'()
   n10 = F_MED_Angel_Backpack_Wings_AddPoseAsset : base(PoseWeight=1) {poseAsset=&F_MED_Angel_Backpack_Wings_AddPoseAsset, poseName=base}
   n11 = Blend Poses by bool(
    BlendPose_0=Apply Additive(Base=Apply Additive(Base=@n6, Additive=@n5, Alpha=0.75) {alphaBoolBlend {blendOption=Linear}}, Additive=@n7) {alphaBoolBlend {blendOption=Linear}}
    BlendPose_1=Apply Additive(Base=Apply Additive(Base=@n9, Additive=@n8, Alpha=0.75) {alphaBoolBlend {blendOption=Linear}}, Additive=@n10) {alphaBoolBlend {blendOption=Linear}}
    BlendTime_0=0.2
    BlendTime_1=0.2
    bActiveValue=Get bIsTargeting()
   ) {blendType=Linear}
   n12 = Get bIsCrouching()
   Output Pose(Result=Blend Poses by bool(
     BlendPose_0=Use cached pose 'preSprint'()
     BlendPose_1=Blend Poses by bool(BlendPose_0=@n11, BlendPose_1=@n4, BlendTime_0=0.3, BlendTime_1=0.3, bActiveValue=@n12) {blendType=Linear}
     BlendTime_0=0.3
     BlendTime_1=0.3
     bActiveValue=Get bIsSkyDiving()
    ) {blendType=Linear}) {layerGroup=None}
   AnimGraphNode_StateMachine_0 = wingPosition() {is=title AnimGraphNode_StateMachine}
   n13 = RigidBody(
    ComponentPose=Local To Component(LocalPose=@AnimGraphNode_StateMachine_0)
    Alpha=Get WingRigidAlpha() -> (WingRigidAlpha:float)
    ExternalForce=Get FallingAnimDynamicsUpForce() -> (FallingAnimDynamicsUpForce:Vector)
   ) {
    bOverrideWorldGravity=true
    overridePhysicsAsset=&/F_MED_Angel_Backpack_Physics
    simulationSpace=BaseBoneSpace
    alphaBoolBlend {blendOption=Linear}
    alphaScaleBiasClamp {bInterpResult=true}
    baseBoneRef {boneName=backpack_root}
    componentAppliedLinearAccClamp=<100 100 100>
    componentLinearAccScale=<1 1 1>
    componentLinearVelScale=<1 1 1>
    overrideWorldGravity {z=-980}
   }
   Save cached pose 'preSprint'(Pose=Component To Local(ComponentPose=@n13)) {cacheName=preSprint}
  ]}
 wingPosition {
  entry=Base
  path=AnimGraph.AnimGraphNode_StateMachine_0.wingPosition
  ids=id
  nodes [
   AnimStateNode_0 = Base() {is=title AnimStateNode}
   AnimStateNode_1 = Diving() {is=title AnimStateNode}
   AnimStateNode_3 = Targeting() {is=title AnimStateNode}
   AnimStateNode_4 = Floating() {is=title AnimStateNode}
   AnimStateNode_5 = Hanging() {is=title AnimStateNode}
   AnimStateTransitionNode_0 = AnimStateTransitionNode()
   AnimStateTransitionNode_1 = AnimStateTransitionNode()
   AnimStateTransitionNode_10 = AnimStateTransitionNode()
   AnimStateTransitionNode_11 = AnimStateTransitionNode()
   AnimStateTransitionNode_12 = AnimStateTransitionNode()
   AnimStateTransitionNode_13 = AnimStateTransitionNode()
   AnimStateTransitionNode_14 = AnimStateTransitionNode()
   AnimStateTransitionNode_2 = AnimStateTransitionNode()
   AnimStateTransitionNode_3 = AnimStateTransitionNode()
   AnimStateTransitionNode_4 = AnimStateTransitionNode()
   AnimStateTransitionNode_5 = AnimStateTransitionNode()
   AnimStateTransitionNode_6 = AnimStateTransitionNode()
   AnimStateTransitionNode_7 = AnimStateTransitionNode()
   AnimStateTransitionNode_9 = AnimStateTransitionNode()
  ]
  links [
   {
    dst=Diving
    rule=AnimGraph.AnimGraphNode_StateMachine_0.wingPosition.AnimStateTransitionNode_0.Transition
    src=Base
    via=AnimStateTransitionNode_0
   }
   {
    dst=Floating
    rule=AnimGraph.AnimGraphNode_StateMachine_0.wingPosition.AnimStateTransitionNode_4.Transition
    src=Base
    via=AnimStateTransitionNode_4
   }
   {
    dst=Hanging
    rule=AnimGraph.AnimGraphNode_StateMachine_0.wingPosition.AnimStateTransitionNode_11.Transition
    src=Base
    via=AnimStateTransitionNode_11
   }
   {
    dst=Targeting
    rule=AnimGraph.AnimGraphNode_StateMachine_0.wingPosition.AnimStateTransitionNode_6.Transition
    src=Base
    via=AnimStateTransitionNode_6
   }
   {
    dst=Base
    rule=AnimGraph.AnimGraphNode_StateMachine_0.wingPosition.AnimStateTransitionNode_3.Transition
    src=Diving
    via=AnimStateTransitionNode_3
   }
   {
    dst=Floating
    rule=AnimGraph.AnimGraphNode_StateMachine_0.wingPosition.AnimStateTransitionNode_1.Transition
    src=Diving
    via=AnimStateTransitionNode_1
   }
   {
    dst=Hanging
    rule=AnimGraph.AnimGraphNode_StateMachine_0.wingPosition.AnimStateTransitionNode_10.Transition
    src=Diving
    via=AnimStateTransitionNode_10
   }
   {
    dst=Base
    rule=AnimGraph.AnimGraphNode_StateMachine_0.wingPosition.AnimStateTransitionNode_2.Transition
    src=Floating
    via=AnimStateTransitionNode_2
   }
   {
    dst=Diving
    rule=AnimGraph.AnimGraphNode_StateMachine_0.wingPosition.AnimStateTransitionNode_5.Transition
    src=Floating
    via=AnimStateTransitionNode_5
   }
   {
    dst=Hanging
    rule=AnimGraph.AnimGraphNode_StateMachine_0.wingPosition.AnimStateTransitionNode_9.Transition
    src=Floating
    via=AnimStateTransitionNode_9
   }
   {
    dst=Base
    rule=AnimGraph.AnimGraphNode_StateMachine_0.wingPosition.AnimStateTransitionNode_12.Transition
    src=Hanging
    via=AnimStateTransitionNode_12
   }
   {
    dst=Diving
    rule=AnimGraph.AnimGraphNode_StateMachine_0.wingPosition.AnimStateTransitionNode_14.Transition
    src=Hanging
    via=AnimStateTransitionNode_14
   }
   {
    dst=Floating
    rule=AnimGraph.AnimGraphNode_StateMachine_0.wingPosition.AnimStateTransitionNode_13.Transition
    src=Hanging
    via=AnimStateTransitionNode_13
   }
   {
    dst=Base
    rule=AnimGraph.AnimGraphNode_StateMachine_0.wingPosition.AnimStateTransitionNode_7.Transition
    src=Targeting
    via=AnimStateTransitionNode_7
   }
  ]
 }
 Base {
  path=AnimGraph.AnimGraphNode_StateMachine_0.wingPosition.AnimStateNode_0.Base
  nodes [Output Animation Pose(Result=F_MED_Angel_Backpack_Wings_PoseAsset : base!(PoseWeight:float=1) -> (Pose:Pose Link Structure) {
     is=title AnimGraphNode_PoseByName
     poseAsset=&/F_MED_Angel_Backpack_Wings_PoseAsset
     poseName=base
    })]
 }
 Diving {
  path=AnimGraph.AnimGraphNode_StateMachine_0.wingPosition.AnimStateNode_1.Diving
  nodes [Output Animation Pose(Result=F_MED_Angel_Backpack_Wings_PoseAsset : dive!(PoseWeight:float=1) -> (Pose:Pose Link Structure) {
     is=title AnimGraphNode_PoseByName
     poseAsset=&F_MED_Angel_Backpack_Wings_PoseAsset
     poseName=dive
    })]
 }
 Targeting {
  path=AnimGraph.AnimGraphNode_StateMachine_0.wingPosition.AnimStateNode_3.Targeting
  nodes [Output Animation Pose(Result=F_MED_Angel_Backpack_Wings_PoseAsset : back!(PoseWeight:float=1) -> (Pose:Pose Link Structure) {
     is=title AnimGraphNode_PoseByName
     poseAsset=&F_MED_Angel_Backpack_Wings_PoseAsset
     poseName=back
    })]
 }
 Floating {
  path=AnimGraph.AnimGraphNode_StateMachine_0.wingPosition.AnimStateNode_4.Floating
  nodes [Output Animation Pose(Result=F_MED_Angel_Backpack_Wings_PoseAsset : fly!(PoseWeight:float=1) -> (Pose:Pose Link Structure) {
     is=title AnimGraphNode_PoseByName
     poseAsset=&F_MED_Angel_Backpack_Wings_PoseAsset
     poseName=fly
    })]
 }
 Hanging {
  path=AnimGraph.AnimGraphNode_StateMachine_0.wingPosition.AnimStateNode_5.Hanging
  nodes [Output Animation Pose(Result=F_MED_Angel_Backpack_Wings_PoseAsset : fly(PoseWeight=1) {poseAsset=&F_MED_Angel_Backpack_Wings_PoseAsset, poseName=fly})]
 }
 Transition {
  path=AnimGraph.AnimGraphNode_StateMachine_0.wingPosition.AnimStateTransitionNode_0.Transition
  nodes [Result(bCanEnterTransition=Get bIsSkydiveDiveMode())]
 }
 Transition {
  path=AnimGraph.AnimGraphNode_StateMachine_0.wingPosition.AnimStateTransitionNode_1.Transition
  nodes [Result(bCanEnterTransition=Get bIsSkydiveFloating())]
 }
 Transition {
  path=AnimGraph.AnimGraphNode_StateMachine_0.wingPosition.AnimStateTransitionNode_10.Transition
  nodes [Result(bCanEnterTransition=Get bIsParachuteOpened())]
 }
 Transition {
  path=AnimGraph.AnimGraphNode_StateMachine_0.wingPosition.AnimStateTransitionNode_11.Transition
  nodes [Result(bCanEnterTransition=Get bIsParachuteOpened())]
 }
 Transition {
  path=AnimGraph.AnimGraphNode_StateMachine_0.wingPosition.AnimStateTransitionNode_12.Transition
  nodes [Result(bCanEnterTransition=NOT Boolean(A=Get bIsSkyDiving()))]
 }
 Transition {
  path=AnimGraph.AnimGraphNode_StateMachine_0.wingPosition.AnimStateTransitionNode_13.Transition
  nodes [Result(bCanEnterTransition=Get bIsSkydiveFloating())]
 }
 Transition {
  path=AnimGraph.AnimGraphNode_StateMachine_0.wingPosition.AnimStateTransitionNode_14.Transition
  nodes [Result(bCanEnterTransition=Get bIsSkydiveDiveMode())]
 }
 Transition {
  path=AnimGraph.AnimGraphNode_StateMachine_0.wingPosition.AnimStateTransitionNode_2.Transition
  nodes [Result(bCanEnterTransition=NOT Boolean(A=Get bIsSkyDiving()))]
 }
 Transition {
  path=AnimGraph.AnimGraphNode_StateMachine_0.wingPosition.AnimStateTransitionNode_3.Transition
  nodes [
   Result(bCanEnterTransition=NOT Boolean(A=Get bIsSkydiveDiveMode()))
   OR Boolean(A=false, B=Get bIsParachuteOpened())
  ]
 }
 Transition {
  path=AnimGraph.AnimGraphNode_StateMachine_0.wingPosition.AnimStateTransitionNode_4.Transition
  nodes [Result(bCanEnterTransition=Get bIsSkydiveFloating())]
 }
 Transition {
  path=AnimGraph.AnimGraphNode_StateMachine_0.wingPosition.AnimStateTransitionNode_5.Transition
  nodes [Result(bCanEnterTransition=Get bIsSkydiveDiveMode())]
 }
 Transition {
  path=AnimGraph.AnimGraphNode_StateMachine_0.wingPosition.AnimStateTransitionNode_6.Transition
  nodes [Result(bCanEnterTransition=Get bIsTargeting())]
 }
 Transition {
  path=AnimGraph.AnimGraphNode_StateMachine_0.wingPosition.AnimStateTransitionNode_7.Transition
  nodes [Result(bCanEnterTransition=NOT Boolean(A=Get bIsTargeting()))]
 }
 Transition {
  path=AnimGraph.AnimGraphNode_StateMachine_0.wingPosition.AnimStateTransitionNode_9.Transition
  nodes [Result(bCanEnterTransition=Get bIsParachuteOpened())]
 }
 EventGraph {links [], nodes []}
]
variables []
